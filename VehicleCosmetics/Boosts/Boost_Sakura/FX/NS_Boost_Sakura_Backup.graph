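# Niagara system for the Sakura vehicle boost cosmetic, named as a backup, emitting flower sprites and ribbon trails from skeletal mesh sockets.
# Flowers spawn in three groups staggered along -X per socket, spinning under curl noise, while two-ribbon trails stream backward.
# Both inherit decaying source movement, with user parameters switching paint coloring and frontend-preview tuning.
# Disabled duplicates of both emitters carry a 150-unit offset.
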NiagaraSystem /VehicleCosmetics/Boosts/Boost_Sakura/FX/NS_Boost_Sakura_Backup
dependencies {
 usedDataInterfaces [
  &NiagaraDataInterfaceArrayFloat
  &NiagaraDataInterfaceCurve
  &NiagaraDataInterfaceVectorCurve
  &NiagaraDataInterfaceSkeletalMesh
 ]
 usedDynamicInputs []
 usedModules [
  &self:SystemSpawnScript.SystemScriptSource.SystemScriptGraph.NiagaraNodeAssignment_0.SetVariables_#1
  &SystemState
 ]
}
emitters {
 Flowers_Opt {
  emitterUpdateScript [
   &EmitterState()
   &VC_FilteredSocketCount()
   &SpawnRate(SpawnRate=15, Use Spawn Probability=false)
   &SpawnRate(SpawnGroup=1, SpawnRate=20, Use Spawn Probability=false) {module=SpawnRate001}
   &SpawnRate(SpawnGroup=2, SpawnRate=20, Use Spawn Probability=false) {module=SpawnRate002}
  ]
  particleSpawnScript [
   &VC_PaintedVariantsOverride(
    Black=<rgba 0.01 0.01 0.01 1>
    BurntSienna=<rgba 0.151 0.0629 0.0189 1>
    Cobalt=<rgba 0.035 0.035 0.7 1>
    Crimson=<rgba 0.4375 0.0456 0.0562 1>
    DefaultColor=<rgba 0.45 0.0675 0.0675 1>
    ForestGreen=<rgba 0 0.2 0.0667 1>
    Grey=<rgba 0.135 0.135 0.135 1>
    Lime=<rgba 0.225 0.3 0 1>
    Orange=<rgba 0.75 0.15 0 1>
    Pink=<rgba 0.6 0.21 0.4582 1>
    Purple=<rgba 0.2139 0.0412 0.7188 1>
    Saffron=<rgba 0.5 0.3833 0 1>
    SkyBlue=<rgba 0.0964 0.1973 0.5 1>
    TitaniumWhite=<rgba 1 1 1 1>
   )
   /SetVariables(Particles.NewFlagsTags:NiagaraInt32=&SpawnGroup()) {isSetParameters=true, module=SetVariables_#2}
   &InitializeParticle(
    Color Mode=::NewEnumerator0
    Lifetime=/VehicleCosmetics/FX/Niagara/Modules/MergeEmitters/VC_FloatBySpawnGroup(
     SpawnGroup or ExecIndex:NiagaraInt32=0
     CustomInt:NiagaraInt32=0
     MaximumSpawnGroups:NiagaraInt32=3
     Float SG 0:NiagaraFloat=&RandomRangeFloat(Minimum=0.3, Maximum=0.8)
     Float SG 1:NiagaraFloat=&RandomRangeFloat(Minimum=0.3, Maximum=0.8)
     Float SG 2:NiagaraFloat=&RandomRangeFloat(Minimum=0.3, Maximum=0.8)
     Float SG 3:NiagaraFloat=0
     Float SG 4:NiagaraFloat=0
     Float SG 5:NiagaraFloat=0
    )
    Sprite Rotation Angle=&RandomRangeFloat(Minimum=-1, Maximum=1)
    Sprite Rotation Mode=::NewEnumerator1
    Sprite Size Mode=::NewEnumerator1
    Uniform Sprite Size Max=16.8
    Uniform Sprite Size Min=7
    UsePositionOffset=false
    Write Material Random=true
   )
   &PartitionParticles(Number Of Partitions=$Emitter.PortCount)
   &SkeletalMeshLocation(
    Bone / Socket Index=$Particles.PartitionParticles.Partition
    Bone / Socket Sampling Mode=::NewEnumerator5
    MaskSpawnGroup=false
    Mesh Sampling Type=::NewEnumerator6
    Sampled Position Offset=&AddVectorToPosition(
     Position=/VehicleCosmetics/FX/Niagara/Modules/MergeEmitters/VC_VectorBySpawnGroup(
      SpawnGroup or ExecIndex:NiagaraInt32=0
      CustomInt:NiagaraInt32=0
      MaximumSpawnGroups:NiagaraInt32=3
      Vector Group 0:Vector3f=<-50 0 0>
      Vector Group 1:Vector3f=<-180 0 0>
      Vector Group 2:Vector3f=<-320 0 0>
      Vector Group 3:Vector3f=<0 0 0>
      Vector Group 4:Vector3f=<0 0 0>
      Vector Group 5:Vector3f=<0 0 0>
     )
     Vector=$User.DTCarOffset
    )
    Skeletal Mesh=$System.SkeletalMesh
   )
   &AddVelocity(Velocity=&RandomRangeVector(Minimum=<-30 0 8>, Maximum=<-60 1 20>), Velocity Speed Scale=2) {enabled=false}
   &AddVelocity(
    Velocity=&MakeCustomVectorFromBool(
     Bool=$User.bIsFrontendPreview
     True Vector=&RandomRangeVector(Minimum=<-30 0 8>, Maximum=<-60 1 20>)
     False Vector=&RandomRangeVector(Minimum=<-30 0 8>, Maximum=<-60 0 20>)
    )
    Velocity Speed Scale=2
   ) {module=AddVelocity001}
   &Color(Color=&MakeCustomLinearColorFromBool(Bool=$User.bIsPainted, True Color=$User.PaintedColor, False Color=<rgba 0.5 0.1 0.1 1>)) {enabled=false}
   &ShapeLocation(
    Cylinder Height=50
    Cylinder Radius=15
    MaskSpawnGroup=false
    Override Seed=false
    Rotation Mode=::NewEnumerator2
    Shape Primitive=::NewEnumerator1
    Surface Only=false
    UseOwnerScale=true
    Yaw / Pitch / Roll=<0.25 0.25 0.25>
   )
  ]
  particleUpdateScript [
   &ParticleState()
   &SpriteRotationRate(Rotation Rate=&RandomRangeFloat(Minimum=360, Maximum=720))
   &ScaleSpriteSize(Uniform Curve Sprite Scale=DataInterface() {propertyValues {Curve {keys [{arriveTangent=0.596, interpMode=RCIM_Cubic, leaveTangent=0.596, tangentMode=RCTM_User, time=0, value=0.8}, <at 1 1 RCIM_Cubic>]}}})
   &AccelerationForce(Acceleration=&RandomRangeVector(Minimum=<-10 0 -10>, Maximum=<0 0 -35>)) {enabled=false}
   &CurlNoiseForce(
    Noise Frequency=&RandomRangeFloat(Minimum=-50, Maximum=50)
    Noise Strength=&RandomRangeFloat(Minimum=-200, Maximum=200)
    Pan Noise=false
   )
   &Drag(
    Drag=&MakeCustomFloatFromBool(Bool=$User.bIsFrontendPreview, True Float=0, False Float=5)
    Use Linear Drag=true
    Use Rotational Drag=true
   )
   &InheritSourceMovement(
    Applied Position Scale=&Multiply_VectorByFloat(
     Vector=<1 1 1>
     Float=&FloatFromCurve(FloatCurve=DataInterface() {propertyValues {Curve {keys [
          {interpMode=RCIM_Cubic, leaveTangent=-0.0051, tangentMode=RCTM_Break, time=0, value=1}
          {arriveTangent=-0.247, interpMode=RCIM_Cubic, tangentMode=RCTM_Break, time=1, value=0.9}
         ]}}})
    )
    ApplyForce=false
    ApplyPosition=true
    ApplyPositionFalloff=false
   )
   &SolveForcesAndVelocity(Clamp Velocity=false, Limit Acceleration=false)
   &DynamicMaterialParameters(
    Index 0 Param 1=1
    Index 0 Param 2=1
    Module.Index0Float0Write=true
    Module.Index0Float1Write=true
    Module.Index0Float2Write=true
    Module.Index0Float3Write=true
   )
   &ScaleColor(
    Scale Alpha=&MakeCustomFloatFromBool(
     Bool=$User.bIsPainted
     True Float=&FloatFromCurve(FloatCurve=DataInterface() {propertyValues {Curve {keys [
          {interpMode=RCIM_Cubic, tangentMode=RCTM_User, time=0, value=0}
          {arriveTangent=-0.3728, interpMode=RCIM_Cubic, leaveTangent=-0.3728, tangentMode=RCTM_User, time=0.2, value=1}
          {arriveTangent=0.1567, interpMode=RCIM_Cubic, leaveTangent=0.1567, tangentMode=RCTM_User, time=0.5, value=1}
          {interpMode=RCIM_Cubic, tangentMode=RCTM_User, time=1, value=0}
         ]}}})
     False Float=&FloatFromCurve(FloatCurve=DataInterface() {propertyValues {Curve {keys [
          {interpMode=RCIM_Cubic, tangentMode=RCTM_User, time=0, value=0}
          {arriveTangent=-0.3728, interpMode=RCIM_Cubic, leaveTangent=-0.3728, tangentMode=RCTM_User, time=0.2, value=1}
          {arriveTangent=0.1567, interpMode=RCIM_Cubic, leaveTangent=0.1567, tangentMode=RCTM_User, time=0.5, value=1}
          {interpMode=RCIM_Cubic, tangentMode=RCTM_User, time=1, value=0}
         ]}}})
    )
    Scale RGB=&MakeCustomVectorFromBool(
     Bool=$User.bIsPainted
     True Vector=&VectorFromCurve(VectorCurve=DataInterface() {propertyValues {
        XCurve {keys [<at 0 6 RCIM_Cubic>,<at 0.45 25 RCIM_Cubic>]}
        YCurve {keys [<at 0 6 RCIM_Cubic>,<at 0.45 25 RCIM_Cubic>]}
        ZCurve {keys [<at 0 6 RCIM_Cubic>,<at 0.45 25 RCIM_Cubic>]}
       }})
     False Vector=&VectorFromCurve(VectorCurve=DataInterface() {propertyValues {
        XCurve {keys [<at 0.17 6 RCIM_Cubic>,<at 0.45 25 RCIM_Cubic>]}
        YCurve {keys [<at 0.17 6 RCIM_Cubic>,<at 0.45 6 RCIM_Cubic>]}
        ZCurve {keys [<at 0.17 6 RCIM_Cubic>,<at 0.45 17 RCIM_Cubic>]}
       }})
    )
    ScaleA=true
    ScaleRGB=true
   ) {module=ScaleColor001}
  ]
  rendererClasses [&NiagaraSpriteRendererProperties]
  renderers [
   {rendererClass=&NiagaraSpriteRendererProperties}
   {rendererClass=&NiagaraSpriteRendererProperties, rendererIndex=1}
   {rendererClass=&NiagaraSpriteRendererProperties, rendererIndex=2}
  ]
 }
 Flowers_Opt001 {
  bEnabled=false
  emitterUpdateScript [
   &EmitterState()
   &VC_FilteredSocketCount()
   &SpawnRate(SpawnRate=15, Use Spawn Probability=false)
   &SpawnRate(SpawnGroup=1, SpawnRate=20, Use Spawn Probability=false) {module=SpawnRate001}
   &SpawnRate(SpawnGroup=2, SpawnRate=20, Use Spawn Probability=false) {module=SpawnRate002}
  ]
  particleSpawnScript [
   &VC_PaintedVariantsOverride() {enabled=false}
   /SetVariables(Particles.NewFlagsTags:NiagaraInt32=&SpawnGroup()) {isSetParameters=true, module=SetVariables_#2}
   &InitializeParticle(
    Color Mode=::NewEnumerator0
    Lifetime=&VC_FloatBySpawnGroup(
     SpawnGroup or ExecIndex:NiagaraInt32=0
     CustomInt:NiagaraInt32=0
     MaximumSpawnGroups:NiagaraInt32=3
     Float SG 0:NiagaraFloat=&RandomRangeFloat(Minimum=0.3, Maximum=0.8)
     Float SG 1:NiagaraFloat=&RandomRangeFloat(Minimum=0.3, Maximum=0.8)
     Float SG 2:NiagaraFloat=&RandomRangeFloat(Minimum=0.3, Maximum=0.8)
     Float SG 3:NiagaraFloat=0
     Float SG 4:NiagaraFloat=0
     Float SG 5:NiagaraFloat=0
    )
    Sprite Rotation Angle=&RandomRangeFloat(Minimum=-1, Maximum=1)
    Sprite Rotation Mode=::NewEnumerator1
    Sprite Size Mode=::NewEnumerator1
    Uniform Sprite Size Max=16.8
    Uniform Sprite Size Min=7
    UsePositionOffset=false
    Write Material Random=true
   )
   &PartitionParticles(Number Of Partitions=$Emitter.PortCount)
   &SkeletalMeshLocation(
    Bone / Socket Index=$Particles.PartitionParticles.Partition
    Bone / Socket Sampling Mode=::NewEnumerator5
    MaskSpawnGroup=false
    Mesh Sampling Type=::NewEnumerator6
    Sampled Position Offset=&AddVectorToPosition(
     Position=&VC_VectorBySpawnGroup(
      SpawnGroup or ExecIndex:NiagaraInt32=0
      CustomInt:NiagaraInt32=0
      MaximumSpawnGroups:NiagaraInt32=3
      Vector Group 0:Vector3f=<-50 150 0>
      Vector Group 1:Vector3f=<-180 150 0>
      Vector Group 2:Vector3f=<-320 150 0>
      Vector Group 3:Vector3f=<0 0 0>
      Vector Group 4:Vector3f=<0 0 0>
      Vector Group 5:Vector3f=<0 0 0>
     )
     Vector=$User.DTCarOffset
    )
    Skeletal Mesh=$System.SkeletalMesh
   )
   &AddVelocity(Velocity=&RandomRangeVector(Minimum=<-30 0 8>, Maximum=<-60 1 20>), Velocity Speed Scale=2) {enabled=false}
   &AddVelocity(
    Velocity=&MakeCustomVectorFromBool(
     Bool=$User.bIsFrontendPreview
     True Vector=&RandomRangeVector(Minimum=<-30 0 8>, Maximum=<-60 1 20>)
     False Vector=&RandomRangeVector(Minimum=<-30 0 8>, Maximum=<-60 0 20>)
    )
    Velocity Speed Scale=2
   ) {module=AddVelocity001}
   &Color(Color=&MakeCustomLinearColorFromBool(Bool=$User.bIsPainted, True Color=$User.PaintedColor, False Color=<rgba 0.5 0.1 0.1 1>))
   &ShapeLocation(
    Cylinder Height=50
    Cylinder Radius=15
    MaskSpawnGroup=false
    Override Seed=false
    Rotation Mode=::NewEnumerator2
    Shape Primitive=::NewEnumerator1
    Surface Only=false
    UseOwnerScale=true
    Yaw / Pitch / Roll=<0.25 0.25 0.25>
   )
  ]
  particleUpdateScript [
   &ParticleState()
   &SpriteRotationRate(Rotation Rate=&RandomRangeFloat(Minimum=360, Maximum=720))
   &ScaleSpriteSize(Uniform Curve Sprite Scale=DataInterface() {propertyValues {Curve {keys [{arriveTangent=0.596, interpMode=RCIM_Cubic, leaveTangent=0.596, tangentMode=RCTM_User, time=0, value=0.8}, <at 1 1 RCIM_Cubic>]}}})
   &AccelerationForce(Acceleration=&RandomRangeVector(Minimum=<-10 0 -10>, Maximum=<0 0 -35>))
   &CurlNoiseForce(
    Noise Frequency=&RandomRangeFloat(Minimum=-50, Maximum=50)
    Noise Strength=&RandomRangeFloat(Minimum=-200, Maximum=200)
    Pan Noise=false
   )
   &Drag(
    Drag=&MakeCustomFloatFromBool(Bool=$User.bIsFrontendPreview, True Float=0, False Float=5)
    Use Linear Drag=true
    Use Rotational Drag=true
   )
   &InheritSourceMovement(
    Applied Position Scale=&Multiply_VectorByFloat(
     Vector=<1 1 1>
     Float=&FloatFromCurve(FloatCurve=DataInterface() {propertyValues {Curve {keys [
          {interpMode=RCIM_Cubic, leaveTangent=-0.0051, tangentMode=RCTM_Break, time=0, value=1}
          {arriveTangent=-0.247, interpMode=RCIM_Cubic, tangentMode=RCTM_Break, time=1, value=0.9}
         ]}}})
    )
    ApplyForce=false
    ApplyPosition=true
    ApplyPositionFalloff=false
   )
   &SolveForcesAndVelocity(Clamp Velocity=false, Limit Acceleration=false)
   &DynamicMaterialParameters(
    Index 0 Param 1=&MakeCustomFloatFromBool(Bool=$User.bIsPainted, True Float=5, False Float=5)
    Index 0 Param 2=&MakeCustomFloatFromBool(Bool=$User.bIsPainted, True Float=1, False Float=0)
    Module.Index0Param0Write=true
    Module.Index0Param1Write=true
    Module.Index0Param2Write=true
    Module.Index0Param3Write=true
   )
   &ScaleColor(
    Scale Alpha=&MakeCustomFloatFromBool(
     Bool=$User.bIsPainted
     True Float=&FloatFromCurve(FloatCurve=DataInterface() {propertyValues {Curve {keys [
          {interpMode=RCIM_Cubic, tangentMode=RCTM_User, time=0, value=0}
          {arriveTangent=-0.3728, interpMode=RCIM_Cubic, leaveTangent=-0.3728, tangentMode=RCTM_User, time=0.2, value=1}
          {arriveTangent=0.1567, interpMode=RCIM_Cubic, leaveTangent=0.1567, tangentMode=RCTM_User, time=0.5, value=1}
          {interpMode=RCIM_Cubic, tangentMode=RCTM_User, time=1, value=0}
         ]}}})
     False Float=&FloatFromCurve(FloatCurve=DataInterface() {propertyValues {Curve {keys [
          {interpMode=RCIM_Cubic, tangentMode=RCTM_User, time=0, value=0}
          {arriveTangent=-0.3728, interpMode=RCIM_Cubic, leaveTangent=-0.3728, tangentMode=RCTM_User, time=0.2, value=1}
          {arriveTangent=0.1567, interpMode=RCIM_Cubic, leaveTangent=0.1567, tangentMode=RCTM_User, time=0.5, value=1}
          {interpMode=RCIM_Cubic, tangentMode=RCTM_User, time=1, value=0}
         ]}}})
    )
    Scale RGB=&MakeCustomVectorFromBool(
     Bool=$User.bIsPainted
     True Vector=&VectorFromCurve(VectorCurve=DataInterface() {propertyValues {
        XCurve {keys [<at 0 6 RCIM_Cubic>,<at 0.45 25 RCIM_Cubic>]}
        YCurve {keys [<at 0 6 RCIM_Cubic>,<at 0.45 25 RCIM_Cubic>]}
        ZCurve {keys [<at 0 6 RCIM_Cubic>,<at 0.45 25 RCIM_Cubic>]}
       }})
     False Vector=&VectorFromCurve(VectorCurve=DataInterface() {propertyValues {
        XCurve {keys [<at 0.17 6 RCIM_Cubic>,<at 0.45 25 RCIM_Cubic>]}
        YCurve {keys [<at 0.17 6 RCIM_Cubic>,<at 0.45 6 RCIM_Cubic>]}
        ZCurve {keys [<at 0.17 6 RCIM_Cubic>,<at 0.45 17 RCIM_Cubic>]}
       }})
    )
    ScaleA=true
    ScaleRGB=true
   ) {module=ScaleColor001}
  ]
  rendererClasses [&NiagaraSpriteRendererProperties]
  renderers [
   {rendererClass=&NiagaraSpriteRendererProperties}
   {rendererClass=&NiagaraSpriteRendererProperties, rendererIndex=1}
   {rendererClass=&NiagaraSpriteRendererProperties, rendererIndex=2}
  ]
 }
 Trail_KimonoDefinitive_Temp {
  emitterUpdateScript [
   &EmitterState()
   &VC_FilteredSocketCount()
   &VC_RateQualityHandler(Epic Spawn Rate=30, High Spawn Rate=30, Low Spawn Rate=30, Use Custom Rate?=true)
   &VC_SpawnRibbon_v1(Number of Ribbons=2, Spawn Time=-1, SpawnRate=$Emitter.Spawn Rate)
  ]
  particleSpawnScript [
   &VC_PaintedVariantsOverride() {enabled=false}
   &VC_InitializeRibbon_v1(
    Enable Skeletal Position=true
    Facing Axis:ENiagaraOrientationAxis=::ZAxis
    Lifetime=&Multiply_Float(A=&MakeCustomFloatFromBool(Bool=$User.bIsFrontendPreview, True Float=0.3, False Float=0.15), B=3)
    Offset Axis:ENiagaraOrientationAxis=::XAxis
    Ribbon Width:NiagaraFloat=&Multiply_Float(A=9.2, B=7)
   )
   &AddVelocity(Velocity=&MakeCustomVectorFromBool(Bool=$User.bIsFrontendPreview, True Vector=<-1000 0 0>, False Vector=<-1000 0 0>)) {module=AddVelocity001}
   &Color(
    Color=&MakeCustomLinearColorFromBool(Bool=$User.bIsPainted, True Color=$User.PaintedColor, False Color=<rgba 1 1 1 1>)
    Scale Color=&VectorFromFloat(Value=$User.PaintedScalar)
   )
  ]
  particleUpdateScript [
   &ParticleState()
   &DynamicMaterialParameters(
    Index 0 Param 1=&MakeCustomFloatFromBool(Bool=$User.bIsPainted, True Float=2, False Float=0)
    Index 0 Param 2=&MakeCustomFloatFromBool(Bool=$User.bIsPainted, True Float=1, False Float=0)
    Index 1 Param 1=1
    Index 2 Param 1=3
    Index 2 Param 2=&SelectFloatFromArray(Direct Array Index=$Particles.SpawnGroup, Float Selection Array=DataInterface() {propertyValues {FloatData [0.2,0.05]}})
    Module.Index0Param0Write=true
    Module.Index0Param1Write=false
    Module.Index0Param2Write=true
    Module.Index0Param3Write=true
    Module.Index1Param0Write=true
    Module.Index1Param1Write=true
    Module.Index1Param2Write=true
    Module.Index1Param3Write=true
    Module.Index2Param0Write=true
    Module.Index2Param1Write=true
    Module.Index2Param2Write=true
    Module.Index2Param3Write=true
    Write Parameter Index 1=true
    Write Parameter Index 2=true
   )
   &InheritSourceMovement(
    Applied Position Scale=&Multiply_VectorByFloat(
     Vector=<1 1 1>
     Float=&FloatFromCurve(FloatCurve=DataInterface() {propertyValues {Curve {keys [{interpMode=RCIM_Cubic, tangentMode=RCTM_Break, time=0.3, value=1}, {interpMode=RCIM_Cubic, tangentMode=RCTM_Break, time=1, value=0}]}}})
    )
    ApplyForce=false
    ApplyPosition=true
    ApplyPositionFalloff=false
   )
   &SolveForcesAndVelocity(Clamp Velocity=false, Limit Acceleration=false)
   &VC_UpdateRibbon_v1(Enable Skeletal Position=true)
   &ScaleColor(
    Scale Alpha=&FloatFromCurve(FloatCurve=DataInterface() {propertyValues {Curve {keys [<at 0 1>,<at 0.9 1>,<at 1 -2>]}}})
    Scale RGB=&MakeCustomVectorFromBool(Bool=$User.bIsPainted, True Vector=&VectorFromFloat(Value=4), False Vector=<1 1 1>)
    ScaleA=true
    ScaleRGB=true
   )
  ]
  rendererClasses [&NiagaraRibbonRendererProperties]
  renderers [{rendererClass=&NiagaraRibbonRendererProperties}]
 }
 Trail_KimonoDefinitive_Temp001 {
  bEnabled=false
  emitterUpdateScript [
   &EmitterState()
   &VC_FilteredSocketCount()
   &VC_RateQualityHandler(Epic Spawn Rate=30, High Spawn Rate=30, Low Spawn Rate=30, Use Custom Rate?=true)
   &VC_SpawnRibbon_v1(Number of Ribbons=2, Spawn Time=-1, SpawnRate=$Emitter.Spawn Rate)
  ]
  particleSpawnScript [
   &VC_PaintedVariantsOverride() {enabled=false}
   &VC_InitializeRibbon_v1(
    Enable Skeletal Position=true
    Facing Axis:ENiagaraOrientationAxis=::ZAxis
    Lifetime=&Multiply_Float(A=&MakeCustomFloatFromBool(Bool=$User.bIsFrontendPreview, True Float=0.3, False Float=0.15), B=3)
    Offset Amount=150
    Ribbon Width:NiagaraFloat=&Multiply_Float(A=9.2, B=7)
   )
   &AddVelocity(Velocity=&MakeCustomVectorFromBool(Bool=$User.bIsFrontendPreview, True Vector=<-1000 0 0>, False Vector=<-1000 0 0>)) {module=AddVelocity001}
   &Color(
    Color=&MakeCustomLinearColorFromBool(Bool=$User.bIsPainted, True Color=$User.PaintedColor, False Color=<rgba 1 1 1 1>)
    Scale Color=&VectorFromFloat(Value=$User.PaintedScalar)
   )
  ]
  particleUpdateScript [
   &ParticleState()
   &DynamicMaterialParameters(
    Index 0 Param 1=&MakeCustomFloatFromBool(Bool=$User.bIsPainted, True Float=2, False Float=0)
    Index 0 Param 2=&MakeCustomFloatFromBool(Bool=$User.bIsPainted, True Float=1, False Float=0)
    Index 1 Param 1=1
    Index 2 Param 1=3
    Index 2 Param 2=&SelectFloatFromArray(Direct Array Index=$Particles.SpawnGroup, Float Selection Array=DataInterface() {propertyValues {FloatData [0.2,0.05]}})
    Module.Index0Param0Write=true
    Module.Index0Param1Write=false
    Module.Index0Param2Write=true
    Module.Index0Param3Write=true
    Module.Index1Param0Write=true
    Module.Index1Param1Write=true
    Module.Index1Param2Write=true
    Module.Index1Param3Write=true
    Module.Index2Param0Write=true
    Module.Index2Param1Write=true
    Module.Index2Param2Write=true
    Module.Index2Param3Write=true
    Write Parameter Index 1=true
    Write Parameter Index 2=true
   )
   &InheritSourceMovement(
    Applied Position Scale=&Multiply_VectorByFloat(
     Vector=<1 1 1>
     Float=&FloatFromCurve(FloatCurve=DataInterface() {propertyValues {Curve {keys [{interpMode=RCIM_Cubic, tangentMode=RCTM_Break, time=0.3, value=1}, {interpMode=RCIM_Cubic, tangentMode=RCTM_Break, time=1, value=0}]}}})
    )
    ApplyForce=false
    ApplyPosition=true
    ApplyPositionFalloff=false
   )
   &SolveForcesAndVelocity(Clamp Velocity=false, Limit Acceleration=false)
   &VC_UpdateRibbon_v1(Enable Skeletal Position=true)
   &ScaleColor(
    Scale Alpha=&FloatFromCurve(FloatCurve=DataInterface() {propertyValues {Curve {keys [<at 0 1>,<at 0.9 1>,<at 1 -2>]}}})
    Scale RGB=&MakeCustomVectorFromBool(Bool=$User.bIsPainted, True Vector=&VectorFromFloat(Value=4), False Vector=<1 1 1>)
    ScaleA=true
    ScaleRGB=true
   )
  ]
  rendererClasses [&NiagaraRibbonRendererProperties]
  renderers [{rendererClass=&NiagaraRibbonRendererProperties}]
 }
}
summary {
 systemName=NS_Boost_Sakura_Backup
 emitters [
  {emitterName=Trail_KimonoDefinitive_Temp, rendererClasses [&NiagaraRibbonRendererProperties]}
  {emitterName=Flowers_Opt, rendererClasses [&NiagaraSpriteRendererProperties]}
  {bEnabled=false, emitterName=Trail_KimonoDefinitive_Temp001, rendererClasses [&NiagaraRibbonRendererProperties]}
  {bEnabled=false, emitterName=Flowers_Opt001, rendererClasses [&NiagaraSpriteRendererProperties]}
 ]
 userVariables [
  User.bIsFrontendPreview {type=NiagaraBool, defaultValue=true}
  User.bIsPainted {type=NiagaraBool}
  User.ColorVariantID {type=NiagaraInt32}
  User.DTCarOffset {type=Vector3f, description="", defaultValue {struct=Vector3f, value=<0 0 0>}}
  User.PaintedColor {type=LinearColor, description="", defaultValue {struct=LinearColor, value=<rgba 1 1 1 1>}}
  User.PaintedScalar {type=NiagaraFloat, defaultValue=1, description=""}
  User.PortCount {type=NiagaraInt32, defaultValue=2}
 ]
}
