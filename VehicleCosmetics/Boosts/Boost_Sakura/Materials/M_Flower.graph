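# Unlit translucent Niagara sprite material for flower particles in the Sakura vehicle boost cosmetic.
# It desaturates the flower texture, adds its green channel scaled by dynamic Param1, then tints the result by particle color through an emissive adjustments function.
# Texture alpha, particle alpha and a depth fade function multiply together for the opacity term.
# The dynamic parameter also names an "isPainted?" channel, which the graph leaves unused.
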Material /VehicleCosmetics/Boosts/Boost_Sakura/Materials/M_Flower
graph {
 properties {
  bAutomaticallySetUsageInEditor=false
  bUsedWithNiagaraSprites=true
  blendMode=BLEND_Translucent
  pixelDepthOffsetMode=PDOM_Legacy
  shadingModel=MSM_Unlit
  usedShadingModels=Unlit
 }
 nodes [
  n1 = TextureSampleParameter2D() {parameterName=Texture, texture=/DealershipSharedLibrary/FX/Textures/Flipbook/T_VCVFX_Flower02_RGB}
  n2 = Desaturation(None=@n1.RGB)
  n3 = Multiply(A=@n1.G, B=DynamicParameter() {paramNames [Param1, isPainted?, Param3, Param4]}.Param1)
  n4 = ParticleColor()
  MaterialFunctionCall(Emissive=Multiply(A=Add(A=@n2, B=@n3), B=@n4.RGB)) {materialFunction=&MF_FX_EmissiveAdjustments}
  Multiply(
   A=Multiply(A=@n1.A, B=@n4.A)
   B=MaterialFunctionCall(Opacity In=Constant() {r=20}, DepthFade Distance=Constant() {r=100}) {materialFunction=&MF_VCVFX_DepthFade}.Result
  )
 ]
}
