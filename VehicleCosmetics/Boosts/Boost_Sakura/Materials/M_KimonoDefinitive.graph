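# Translucent, two-sided unlit Niagara ribbon material for the kimono effect in the Sakura vehicle boost.
# Emissive layers kimono texture channels, a panning flower outline and particle color, with a dynamic IsPainted parameter blending tinted and untinted versions through MF_FX_EmissiveAdjustments.
# A time-driven sine on the second UV set, weighted along U, builds a wave vector with adjustable wave count.
# Opacity combines texture and particle alpha with power-curve fades along U.
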Material /VehicleCosmetics/Boosts/Boost_Sakura/Materials/M_KimonoDefinitive
graph {
 properties {
  bAutomaticallySetUsageInEditor=false
  bUsedWithNiagaraRibbons=true
  blendMode=BLEND_Translucent
  pixelDepthOffsetMode=PDOM_Legacy
  shadingModel=MSM_Unlit
  twoSided=true
  usedShadingModels=Unlit
 }
 nodes [
  n1 = Constant() {r=1}
  n2 = ComponentMask(None=TextureCoordinate() {coordinateIndex=1}) {r=true}
  n3 = DynamicParameter() {parameterIndex=2, defaultValue=<rgba 1 0 1 1>, paramNames [BrightPink, Multiplier_CountWaves, Param3, Param4]}
  n4 = Add(A=@n3.Multiplier_CountWaves, B=ScalarParameter() {defaultValue=0.5, parameterName=CountWaves2})
  n5 = ScalarParameter() {defaultValue=7, parameterName=Speed}
  n6 = OneMinus(None=Time())
  n7 = Sine(None=Add(A=Divide(A=@n2, B=@n4), B=Multiply(A=@n5, B=@n6)))
  n8 = ScalarParameter() {defaultValue=5, parameterName=Strenght}
  n9 = Power(Base=ComponentMask(None=TextureCoordinate()) {r=true}, Exp=ScalarParameter() {defaultValue=0.5, parameterName=PowerMiddle2})
  n10 = ScalarParameter() {defaultValue=0.5, parameterName=MultPower2}
  AppendVector(A=AppendVector(A=@n1, B=@n1), B=Multiply(A=Multiply(A=@n7, B=@n8), B=Multiply(A=@n9, B=@n10)))
  n11 = OneMinus(None=ComponentMask(None=TextureCoordinate() {uTiling=0.2}) {r=true})
  n12 = OneMinus(None=ComponentMask(None=TextureCoordinate() {uTiling=0.8}) {r=true})
  n13 = LinearInterpolate(A=Multiply(A=@n11, B=@n11), B=Multiply(A=@n12, B=@n12), Alpha=DynamicParameter() {parameterIndex=1}.R)
  n14 = ParticleColor()
  n15 = TextureSampleParameter2D() {parameterName=MaterialExpressionTextureSampleParameter2D_0, texture=/DealershipSharedLibrary/FX/Textures/TexturePackage/T_VCVFX_Kimono_RGB}
  n16 = Multiply(A=@n14.RGB, B=@n15.R)
  n17 = ScalarParameter() {defaultValue=1.2156, parameterName=Time}
  n18 = Panner(Coordinate=TextureCoordinate() {uTiling=1.5}, Time=Multiply(A=Time(), B=Frac(None=@n17))) {speedX=-1}
  n19 = TextureSampleParameter2D(UVs=@n18) {parameterName=MaterialExpressionTextureSampleParameter2D_0, texture=&T_VCVFX_Kimono_RGB}
  n20 = Multiply(A=@n19.G, B=Constant() {r=1})
  n21 = TextureSample(UVs=@n18) {texture=/DealershipSharedLibrary/FX/Textures/Gradient_Masks/T_VCVFX_FlowerOutline_A}
  n22 = Add(A=Add(A=Multiply(A=@n13, B=@n16), B=Multiply(A=@n14.RGB, B=@n20)), B=Multiply(A=Constant() {r=3.5}, B=Multiply(A=@n21.R, B=@n14.RGB)))
  n23 = Multiply(
   A=Multiply(
    A=VectorParameter() {parameterName=Inner_Color, defaultValue=<rgba 1 0.2 0.2 1>}.RGB
    B=ScalarParameter() {defaultValue=2, parameterName=InnerMultiplyColor}
   ) {constB=2}
   B=@n22
  )
  n24 = ParticleColor()
  MaterialExpressionTextureSampleParameter2D_0 = TextureSampleParameter2D() {parameterName=MaterialExpressionTextureSampleParameter2D_0, texture=&T_VCVFX_Kimono_RGB}
  n25 = Multiply(A=Multiply(A=@n24.RGB, B=@MaterialExpressionTextureSampleParameter2D_0.B), B=@n3.BrightPink)
  n26 = Multiply(
   A=Multiply(
    A=VectorParameter() {parameterName=HL_Color, defaultValue=<rgba 1 0.149 0.4 1>}.RGB
    B=ScalarParameter() {defaultValue=30, parameterName=HLMultiplyColor}
   ) {constB=30}
   B=@n25
  )
  n27 = DynamicParameter() {defaultValue=<rgba 1 0 0 0>, paramNames [IsPainted?, Param2, Param3, Param4]}
  MaterialFunctionCall(Emissive=LinearInterpolate(A=Add(A=@n23, B=@n26) {desc=ColorVersion}, B=Add(A=@n22, B=@n25) {constB=0, desc=BnWVersion}, Alpha=Saturate(None=@n27.IsPainted?))) {materialFunction=&MF_FX_EmissiveAdjustments}
  n28 = Multiply(A=@n14.A, B=@n15.A)
  n29 = ComponentMask(None=TextureCoordinate()) {r=true}
  n30 = Constant() {r=16}
  n31 = Power(Base=Saturate(None=Multiply(A=@n29, B=@n30)), Exp=ScalarParameter() {defaultValue=7, parameterName=FadeInPower})
  n32 = Multiply(A=@MaterialExpressionTextureSampleParameter2D_0.A, B=@n24.A)
  n33 = ComponentMask(None=TextureCoordinate()) {r=true}
  n34 = Constant() {r=12}
  n35 = Power(Base=Saturate(None=Multiply(A=@n33, B=@n34)), Exp=ScalarParameter() {defaultValue=7, parameterName=FadeInPower})
  n36 = TextureCoordinate()
  Multiply(A=Add(A=Multiply(A=@n28, B=@n31), B=Multiply(A=@n32, B=@n35)), B=Power(Base=ComponentMask(None=@n36) {r=true}) {constExponent=0.5})
 ]
}
